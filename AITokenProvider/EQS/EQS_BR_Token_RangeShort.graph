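# EQS query that picks short-range positions for AI, apparently for a Battle Royale token system.
# A composite generator merges points on a circle with token-reserved positions.
# A 3D distance test only filters, keeping points inside a band around the querier.
# An obstruction test against the querier, with both ends raised by height offsets, then both filters and scores the candidates.
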EnvQuery /AITokenProvider/EQS/EQS_BR_Token_RangeShort
options [{
  generator {
   name=EnvQueryGenerator_Composite_0
   class=EnvQueryGenerator_Composite
   properties {
    optionName=""
    generators [
     &self:EnvQueryGenerator_Composite_0.EnvQueryGenerator_OnCircle_0
     &self:EnvQueryGenerator_Composite_0.FortQueryGenerator_TokenReservedPositions_0
    ]
   }
  }
  tests [
   EnvQueryTest_Distance_0 {
    class=EnvQueryTest_Distance
    properties {
     distanceTo=&EnvQueryContext_Querier
     filterType=Range
     floatValueMax=2000
     floatValueMin=1000
     scoringEquation=Linear
     scoringFactor=1
     testComment=""
     testMode=Distance3D
     testPurpose=Filter
    }
   }
   FortQueryTest_IsObstructed_1 {
    class=FortQueryTest_IsObstructed
    properties {context=&EnvQueryContext_Querier, contextHeightOffset=150, itemHeightOffset=150, testPurpose=FilterAndScore}
   }
  ]
 }]
